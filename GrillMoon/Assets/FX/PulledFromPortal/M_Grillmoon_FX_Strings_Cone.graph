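# Translucent, unlit, two-sided Niagara mesh-particle material for a cone of strings in the GrillMoon pulled-from-portal effect.
# Panned turbulence distorts plasma noise for a gradient-driven erosion, shaped by particle alpha, Fresnel and quality-switched depth fade.
# A stencil-tested character mask, built from depth-reconstructed world position, height gradient and particle-centered sphere, adds to that opacity.
# Particle color feeds emissive adjustments, and a camera-space vector weighted toward one UV end provides a top offset.
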Material /GrillMoon/Assets/FX/PulledFromPortal/M_Grillmoon_FX_Strings_Cone
graph {
 properties {
  bUsedWithNiagaraMeshParticles=true
  blendMode=BLEND_Translucent
  pixelDepthOffsetMode=PDOM_Legacy
  shadingModel=MSM_Unlit
  twoSided=true
  usedShadingModels=Unlit
 }
 nodes [
  n1 = ComponentMask(None=TextureCoordinate() {vTiling=1.25}) {g=true}
  n2 = ParticleColor()
  n3 = Multiply(A=ComponentMask(None=TextureCoordinate()) {g=true}, B=ScalarParameter() {parameterName=Gradient}) {constB=4}
  n4 = ScalarParameter() {parameterName=GradientOffset}
  n5 = Add(A=ScalarParameter() {parameterName=Erode_p}, B=Saturate(None=Add(A=@n3, B=@n4)))
  n6 = TextureSample(UVs=Panner(Coordinate=TextureCoordinate() {vTiling=0.6}) {speedY=-0.5}) {texture=/Game/Global/Textures/Default/Caustics/T_Turbulence_Seamless}
  n7 = Add(A=Multiply(A=@n6.R) {constB=0.35}, B=Panner(Coordinate=TextureCoordinate() {uTiling=3, vTiling=3}) {speedY=-1.7})
  n8 = SmoothStep(
   Min=@n5
   Max=Add(A=@n5, B=ScalarParameter() {parameterName=Erode_w})
   Value=Add(A=@n6.R, B=TextureSample(UVs=@n7) {texture=/Game/Effects/Fort_Effects/Textures/Plasma/T_Plasma_03_M}.R)
  )
  n9 = Multiply(A=Power(Base=OneMinus(None=@n1)), B=Multiply(A=@n2.A, B=OneMinus(None=@n8)))
  n10 = ParticleColor()
  n11 = MaterialFunctionCall(Exponent=ScalarParameter() {parameterName=Fresnel}) {materialFunction=&MF_Fresnel_FX}
  n12 = Saturate(None=Multiply(A=Multiply(A=@n9, B=@n10.A), B=@n11.Fresnel Double Sided))
  n13 = TwoSidedSign()
  n14 = DepthFade(
   Opacity=@n12
   FadeDistance=LinearInterpolate(
    A=ScalarParameter() {defaultValue=250, group=Opacity, parameterName=DepthFadeDistance_InnerFaces}
    B=ScalarParameter() {defaultValue=250, group=Opacity, parameterName=DepthFadeDistance}
    Alpha=Saturate(None=@n13)
   )
  )
  n15 = MaterialFunctionCall!(Exclude WPO Offsets) {materialFunction=/Engine/Functions/Engine_MaterialFunctions02/UVs/WorldPositionBehindFromDepth_experimental}
  n16 = ParticlePositionWS()
  n17 = Add(A=Subtract(A=@n15.Result, B=@n16), B=ScalarParameter() {group=CharacterGradient, parameterName=Character_Gradient_offset})
  n18 = ScalarParameter() {defaultValue=2.8392, group=CharacterGradient, parameterName=GradientSize}
  n19 = OneMinus(None=ComponentMask(None=Divide(A=@n17, B=@n18)) {b=true})
  n20 = ScalarParameter() {defaultValue=1, group=CharacterGradient, parameterName=StrengthOfCharMasking}
  n21 = Add(A=@n16, B=VectorParameter() {parameterName=GradientMaskingOffset, defaultValue=<rgba 0 0 0 1>}.RGB)
  n22 = ScalarParameter() {group=CharacterGradient, parameterName=Character_Masking_Depth}
  n23 = ScalarParameter() {defaultValue=255, parameterName=StencilID}
  n24 = Multiply(
   A=Saturate(None=Multiply(A=@n19, B=@n20))
   B=Multiply(
    A=SphereMask(A=@n21, B=@n15.Result, Radius=@n22) {hardnessPercent=0}
    B=MaterialFunctionCall!(ID=@n23) {materialFunction=/Game/Packages/Fortress_SharedMaterials/Stencil_ID_Definitions/StencilMaskCompare}.Result
   )
  )
  Add(A=QualitySwitch(Default=@n12, High=@n14, Epic=@n14), B=@n24)
  MaterialFunctionCall(Emissive=Multiply(A=Constant() {r=1}, B=@n2.RGB)) {materialFunction=&MF_FX_EmissiveAdjustments}
  Multiply(A=@n24, B=@n11.Fresnel Double Sided)
  n25 = ComponentMask(None=TextureCoordinate()) {g=true}
  n26 = VectorParameter() {parameterName=Screenspace_TopOffset, defaultValue=<rgba 0 0 0 1>}
  Multiply(
   A=Saturate(None=Power(Base=@n25) {constExponent=6})
   B=Multiply(A=Transform(None=@n26.RGB) {transformSourceType=TRANSFORMSOURCE_Camera}, B=ScalarParameter() {parameterName=TopViewspace_Offset})
  )
  Saturate()
 ]
}
